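; Part modifier for the ShrimpStroll back bling that drives bursts in its idle VFX from player movement.
; Outside the frontend, post-init caches the player pawn and binds its landed event, which pulses the "LandBurst" particle parameter on, then off after a short delay.
; Tick gives the same pulse on sliding or tactical sprinting, behind a DoOnce with no visible reset.
; The frontend branch only runs a delay into an unconnected FlipFlop.
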(event EventOnPostResIn
  ;[0 0 0] post res
  (|CallParentFunction)
  (if (Reactive|Frontend|IsCosmeticObjectDisplayedinFrontendContext 0)
    (Utilities|FlowControl|Delay 4.0)
    (Utilities|FlowControl|FlipFlop)
    (else
      (bind _asfort_player_pawn (Utilities|Casting|CastToFortPlayerPawn (CustomCharacter|GetAssociatedPlayerPawn)))
      (Variables|Default|SetplayerRef _asfort_player_pawn)
      (EventDispatchers|AssignOnPawnLanded (CustomCharacter|GetAssociatedPlayerPawn) (delegate Custom|OnPawnLanded_Event)))))

(event Custom|OnPawnLanded_Event
  (params (Hit :type Hit%20Result%20Structure%20%28by%20ref%29))
  (bind _returnvalue (CallFunction|GetIdleVFXComponent))
  (Effects|Components|ParticleSystem|SetFloatParameter _returnvalue "LandBurst" 1.0)
  (Utilities|FlowControl|Delay 0.1)
  (Effects|Components|ParticleSystem|SetFloatParameter _returnvalue "LandBurst" 0.0))

(event EventTick
  (params (DeltaSeconds :type Float%20%28single-precision%29))
  (bind _returnvalue (CallFunction|GetIdleVFXComponent))
  ;[0 0 0] tick fx
  (if (Reactive|Frontend|IsCosmeticObjectDisplayedinFrontendContext 0)
    _
    (elif (or (Variables|/Script/FortniteGame.FortPawn|GetbIsSliding (Variables|Default|GetplayerRef)) (Variables|/Script/FortniteGame.FortPlayerPawn|GetbIsTacticalSprinting (Variables|Default|GetplayerRef)))
      (Utilities|FlowControl|DoOnce false
        (:Completed
          (Effects|Components|ParticleSystem|SetFloatParameter _returnvalue "LandBurst" 1.0)
          (Utilities|FlowControl|Delay 0.1)
          (Effects|Components|ParticleSystem|SetFloatParameter _returnvalue "LandBurst" 0.0)))
      (else
        _))))

(fn ConstructionScript
  (|CallParentFunction))
