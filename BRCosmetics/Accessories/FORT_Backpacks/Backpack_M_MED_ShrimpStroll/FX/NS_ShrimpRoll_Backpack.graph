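# Niagara effect for the ShrimpStroll backpack cosmetic that sheds small green leaf sprites from its skeletal mesh.
# Steady spawning thins as owner speed rises and stops while skydiving; User.LandBurst adds quality-scaled bursts, throttled by a timeline that recovers slowly, and alters velocity, lifetime and gravity.
# Leaves inherit owner motion, then flutter through aerodynamic drag, mesh-orientation alignment and sine-driven width scaling; the wind force is multiplied to zero.
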NiagaraSystem /BRCosmetics/Accessories/FORT_Backpacks/Backpack_M_MED_ShrimpStroll/FX/NS_ShrimpRoll_Backpack
dependencies {
 usedDataInterfaces [&NiagaraDataInterfaceArrayFloat, &NiagaraDataInterfaceCurve, &NiagaraDataInterfaceSkeletalMesh]
 usedDynamicInputs [&Multiply_Vector]
 usedModules [
  &self:SystemSpawnScript.SystemScriptSource.SystemScriptGraph.NiagaraNodeAssignment_3.SetVariables_#1
  &SystemState
  &self:SystemSpawnScript.SystemScriptSource.SystemScriptGraph.NiagaraNodeAssignment_8.SetVariables_#2
  &self:SystemSpawnScript.SystemScriptSource.SystemScriptGraph.NiagaraNodeAssignment_9.SetVariables_#3
 ]
}
emitters {Leaves_Falling {
  emitterSpawnScript [/SetVariables(Emitter.Delay=1) {isSetParameters=true}]
  emitterUpdateScript [
   &EmitterState(
    Inactive Response=::NewEnumerator1
    Life Cycle Mode=::NewEnumerator1
    Loop Behavior=::NewEnumerator1
    Loop Duration Mode=::NewEnumerator1
    UseLoopDelay=false
   )
   &SpawnRate(
    Spawn Probability=&MakeCustomFloatFromBool(Bool=$User.IsSkydiving, True Float=0, False Float=1)
    SpawnRate=&Multiply_Float(
     A=&Lerp_Float(
      A=10
      B=0
      Alpha=&ClampFloat(Float=&Divide_Float(A=&VectorLength(Vector=$Engine.Owner.Velocity), B=500))
     )
     B=$System.SpawnRateCurve
    )
    Use Spawn Probability=true
   )
   &SpawnPerFrame(
    Spawn Count:NiagaraInt32=&MakeIntFromFloat(Float=&Multiply_Float(
      A=&RandomRangeFloat(
       Minimum=&SelectFloatFromArray(Direct Array Index=$Engine.QualityLevel, Float Selection Array=DataInterface() {propertyValues {FloatData [4,5,6,7,7]}})
       Maximum=&SelectFloatFromArray(Direct Array Index=$Engine.QualityLevel, Float Selection Array=DataInterface() {propertyValues {FloatData [6,8,9,10,10]}})
      )
      B=&FloatFromCurve(
       FloatCurve=DataInterface() {propertyValues {Curve {keys [
           {arriveTangent=0.5131, interpMode=RCIM_Cubic, leaveTangent=0.5131, tangentMode=RCTM_User, time=0, value=0}
           {arriveTangent=1.2242, interpMode=RCIM_Cubic, leaveTangent=1.2242, tangentMode=RCTM_User, time=0.5, value=0.3}
           {arriveTangent=1.9063, interpMode=RCIM_Cubic, leaveTangent=1.9063, tangentMode=RCTM_User, time=1, value=1}
          ]}}}
       CurveIndex=$Emitter.Delay
      )
     ))
    Spawn Probability:NiagaraFloat=$User.LandBurst
    Use Spawn Probability=true
   )
   &Timeline(
    Play=&SetBoolByFloatComparison(A=$User.LandBurst, B=1, Comparison Type=::NewEnumerator4)
    Play Rate=20
    Rewind Rate=0.5
   )
   /SetVariables(Emitter.Delay:NiagaraFloat=$Emitter.Timeline.InverseNormalizedTime) {isSetParameters=true}
   /SetVariables(
    Emitter.WindOscillateTime:NiagaraFloat=&Divide_Float(A=$Engine.Time, B=44)
    Emitter.WindRotateTime:NiagaraFloat=&Divide_Float(A=$Engine.Time, B=66)
   ) {isSetParameters=true}
   /SetVariables(
    Emitter.WindIntensity:NiagaraFloat=&Multiply_Float(A=&Sine(Normalized Angle=$Emitter.WindOscillateTime, Period=1, Scale=1, Bias=0), B=0)
    Emitter.WindVector:Vector3f=&MakeVector(
     X=&Add_Float(
      A=&Sine(Normalized Angle=$Emitter.WindRotateTime, Period=1, Scale=1, Bias=0)
      B=&Cosine(Normalized Angle=$Emitter.WindRotateTime, Period=1, Scale=1, Bias=0)
     )
     Y=&Subtract_Float(
      A=&Cosine(Normalized Angle=$Emitter.WindRotateTime, Period=1, Scale=1, Bias=0)
      B=&Sine(Normalized Angle=$Emitter.WindRotateTime, Period=1, Scale=1, Bias=0)
     )
     Z=0
    )
   ) {isSetParameters=true}
  ]
  particleSpawnScript [
   &FN_SetSpawnGroup(Number of Spawn Groups=2)
   &InitializeParticle(
    Color=&MaskLinearColorBySpawnGroup(
     Masked Linear Color=<rgba 0.0626 0.1 0.0138 1>
     Passthrough Linear Color=&RandomRangeLinearColor(Minimum=<rgba 0.1253 0.2 0.0276 1>, Maximum=<rgba 0.1879 0.3 0.0414 1>)
    )
    Lifetime=&Multiply_Float(
     A=&RandomRangeFloat(Minimum=5.666, Maximum=8.333)
     B=&Lerp_Float(A=1, B=0.5, Alpha=$User.LandBurst)
    )
    Mass=&Multiply_Float(
     A=&RandomRangeFloat(Minimum=2, Maximum=4)
     B=&FloatFromCurve(
      FloatCurve=DataInterface() {propertyValues {Curve {keys [<at 0 0.5 RCIM_Cubic>, {arriveTangent=1.5048, interpMode=RCIM_Cubic, leaveTangent=1.5048, tangentMode=RCTM_User, time=1, value=1}]}}}
      CurveIndex=&ReturnNormalizedExecIndex()
     )
    )
    Mass Mode=::NewEnumerator1
    Sprite Rotation Mode=::NewEnumerator2
    Sprite Size Max=<6 3>
    Sprite Size Min=<3 2>
    Sprite Size Mode=::NewEnumerator2
    UsePositionOffset=false
    Write Material Random=true
   )
   &AddVelocity(Velocity=&Lerp_Vector(
     A=&RandomRangeVector(Minimum=<-25 0 50>, Maximum=<25 0 50>)
     B=&RandomRangeVector(Minimum=<-20 -25 50>, Maximum=<20 -50 50>)
     Alpha=$User.LandBurst
    ))
   &InheritVelocity(
    Inherited Velocity Amount Scale=&MakeVector(
     X=&Lerp_Float(A=1, B=0.5, Alpha=$User.LandBurst)
     Y=&Lerp_Float(A=1, B=0.25, Alpha=$User.LandBurst)
     Z=&Lerp_Float(A=1, B=1.25, Alpha=$User.LandBurst)
    )
    Inherited Velocity Speed Limit:NiagaraFloat=&RandomRangeFloat(Minimum=111, Maximum=250)
    Speed Limit=true
    UseSpeedThreshold=false
   )
   &ApplyInitialForces(Apply Force to Position=true)
   &SkeletalMeshLocation(
    MaskSpawnGroup=false
    Mesh Sampling Type=::NewEnumerator5
    Orientation Sampling=::NewEnumerator1
    OverrideRandomSeed=false
    Skeletal Mesh=$System.Skeletal Mesh
   )
   &ShapeLocation(
    MaskSpawnGroup=false
    Non Uniform Scale=<0.75 1 0.5>
    Offset=&Lerp_Vector(A=<0 0 0>, B=<0 0 10>, Alpha=$User.LandBurst)
    Offset Mode=::NewEnumerator0
    Override Seed=false
    Sphere Radius=5
    UseOwnerScale=true
   )
   &CameraOffset(Camera Offset Amount=5)
  ]
  particleUpdateScript [
   &ParticleState()
   &ScaleSpriteSize(
    Scale Factor=&Multiply_Vector2(
     A=&MakeVector2D(
      X=&Multiply_Float(
       A=1
       B=&Sine(
        Normalized Angle=$Emitter.Age
        Period=&Multiply_Float(A=3, B=&Add_Float(A=$Particles.Initial.MaterialRandom, B=0.6))
        Scale=1
        Bias=0
       )
      )
      Y=1
     )
     B=&Vector2DFromFloat(Value=&FloatFromCurve(FloatCurve=DataInterface() {propertyValues {Curve {keys [
           {arriveTangent=16.7921, interpMode=RCIM_Cubic, leaveTangent=16.7921, tangentMode=RCTM_User, time=0, value=0.5}
           {arriveTangent=-1.0851, interpMode=RCIM_Cubic, leaveTangent=-1.0851, tangentMode=RCTM_User, time=0.1, value=1}
           {arriveTangent=-1.2169, interpMode=RCIM_Cubic, leaveTangent=-1.2169, tangentMode=RCTM_User, time=0.5, value=0.55}
           <at 1 0.25>
          ]}}}))
    )
    Scale Sprite Size Mode=::NewEnumerator2
   )
   &ScaleColor(
    Scale Alpha=&FloatFromCurve(FloatCurve=DataInterface() {propertyValues {Curve {keys [<at -0.0046 1.0113 RCIM_Cubic>,<at 1.0011 0.002 RCIM_Cubic>]}}})
    ScaleA=true
    ScaleRGB=false
   )
   &AccelerationForce(Acceleration=&RandomRangeVector(Minimum=<0 0 33>, Maximum=<0 0 66>)) {enabled=false}
   &GravityForce(Gravity=&Multiply_VectorByFloat(
     Vector=<0 0 -666>
     Float=&Lerp_Float(
      A=&RandomRangeFloat(Minimum=0.8, Maximum=1)
      B=&FloatFromCurve(
       FloatCurve=DataInterface() {propertyValues {Curve {keys [<at 0 0.1 RCIM_Cubic>, {arriveTangent=1.5048, interpMode=RCIM_Cubic, leaveTangent=1.5048, tangentMode=RCTM_User, time=1, value=1}]}}}
       CurveIndex=&ReturnNormalizedExecIndex()
       Scale Curve=5
      )
      Alpha=$User.LandBurst
     )
    ))
   &PointForce(
    Force Origin=&AddVectorToPosition(Position=/Niagara/DynamicInputs/Helpers/SimulationPosition(), Vector=<0 0 50>)
    Force Strength=1000
    Module.Use Falloff=true
    Module.UseLinearForce=false
   )
   &InheritSourceMovement(
    Applied Force Scale=&VectorFromFloat(Value=&FloatFromCurve(FloatCurve=DataInterface() {propertyValues {Curve {keys [
          {interpMode=RCIM_Cubic, tangentMode=RCTM_User, time=0, value=1}
          {arriveTangent=-7.0714, interpMode=RCIM_Cubic, leaveTangent=-7.0714, tangentMode=RCTM_User, time=0.3, value=0}
         ]}}}))
    Applied Position Scale=&VectorFromFloat(Value=&FloatFromCurve(
      FloatCurve=DataInterface() {propertyValues {Curve {keys [{interpMode=RCIM_Cubic, tangentMode=RCTM_User, time=0, value=1}, {interpMode=RCIM_Cubic, tangentMode=RCTM_User, time=1, value=0}]}}}
      CurveIndex=$Particles.Age
      Scale Curve=&Lerp_Float(A=0.5, B=0.75, Alpha=$User.LandBurst)
     ))
    ApplyForce=true
    ApplyForceFalloff=false
    ApplyPosition=true
    ApplyPositionFalloff=false
    Module.LimitForceSpeed=false
   )
   &WindForce(Wind Speed=$Emitter.WindVector, Wind Speed Scale=$Emitter.WindIntensity)
   &AerodynamicDrag(
    Drag Shape Curve=DataInterface() {propertyValues {Curve {keys [
        {interpMode=RCIM_Cubic, tangentMode=RCTM_User, time=0, value=0.25}
        {arriveTangent=2.0799, interpMode=RCIM_Cubic, leaveTangent=2.0799, tangentMode=RCTM_User, time=1, value=1}
       ]}}}
    Lift Shape Curve=DataInterface() {propertyValues {Curve {keys [
        {interpMode=RCIM_Cubic, tangentMode=RCTM_User, time=0, value=0}
        {interpMode=RCIM_Cubic, tangentMode=RCTM_User, time=0.2, value=1}
        {arriveTangent=-0.375, interpMode=RCIM_Cubic, leaveTangent=-0.375, time=0.5, value=0.35}
        {tangentMode=RCTM_User, time=1, value=0.7}
       ]}}}
    Pivot Offset Minimum Length:Vector3f=<0.15 0 1>
    Rotation Strength:NiagaraFloat=&FloatFromCurve(
     FloatCurve=DataInterface() {propertyValues {Curve {keys [<at 0 0.1 RCIM_Cubic>, {arriveTangent=1.5048, interpMode=RCIM_Cubic, leaveTangent=1.5048, tangentMode=RCTM_User, time=1, value=1}]}}}
     CurveIndex=&ReturnNormalizedExecIndex(Normalized Index Scale=10)
     Scale Curve=5
    )
    Surface Normal Axis:Vector3f=&RandomRangeVector(Minimum=<-0.5 -0.5 1>, Maximum=<0.5 0.5 1>)
    Use Lift Smoothing=true
    Use Linear Drag=true
    Use Rotational Drag=true
   )
   &SolveForcesAndVelocity(Clamp Velocity=false, Limit Acceleration=false)
   &AlignSpriteToMeshOrientation(SetAlignment=true, SetFacing=true)
   &DynamicMaterialParameters(
    Index 0 Param 1=&RandomRangeFloat(Minimum=0, Maximum=0.5)
    Module.Index0Param0Write=true
    Module.Index0Param1Write=false
    Module.Index0Param2Write=false
    Module.Index0Param3Write=false
   )
   &KillParticlesInVolume(Invert Volume=true, Origin Offset:Vector3f=<0 0 80>, Volume Origin:Vector3f=$Particles.SkeletalMeshLocation.SampledPosition) {enabled=false}
  ]
  rendererClasses [&NiagaraSpriteRendererProperties]
  renderers [{rendererClass=&NiagaraSpriteRendererProperties}]
 }}
summary {
 systemName=NS_ShrimpRoll_Backpack
 emitters [{emitterName=Leaves_Falling, rendererClasses [&NiagaraSpriteRendererProperties]}]
 userVariables [
  User.bIsFrontEnd {type=NiagaraBool}
  User.bIsFrontEndPreview {type=NiagaraBool}
  User.IsSkydiving {type=NiagaraBool}
  User.LandBurst {type=NiagaraFloat, defaultValue=0, description=""}
 ]
}
