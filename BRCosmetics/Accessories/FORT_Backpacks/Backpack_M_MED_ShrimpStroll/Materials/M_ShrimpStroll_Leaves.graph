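# Masked, two-sided subsurface material for the animated leaves on the ShrimpStroll backpack cosmetic.
# Color blends two greens along a local-Z gradient, adding mask- and vertex-color-driven darkening plus a Fresnel lift.
# Opacity thins grazing-angle leaves using normalized Fresnel and HeightLerp, scaled by camera distance and bypassed by vertex green, which also flattens normals.
# World position offset combines vertex-color-weighted sine wobble along normals with gradient-weighted top sway.
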Material /BRCosmetics/Accessories/FORT_Backpacks/Backpack_M_MED_ShrimpStroll/Materials/M_ShrimpStroll_Leaves
graph {
 properties {
  bAutomaticallySetUsageInEditor=false
  bUsedWithMeshDeformer=true
  bUsedWithSkeletalMesh=true
  blendMode=BLEND_Masked
  shadingModel=MSM_Subsurface
  twoSided=true
  usedShadingModels=Subsurface
 }
 nodes [
  CollectionParameter() {collection=&FortniteMaterialParameters, parameterName=PlayerDirection}
  n1 = ScalarParameter() {defaultValue=0.1557, parameterName=MaskTransition}
  Fresnel(ExponentIn=@n1) {baseReflectFraction=0}
  n2 = Divide(
   A=Subtract(A=LocalPosition() {localOrigin=InstancePreSkinning}.Z, B=ScalarParameter() {defaultValue=-17.5, parameterName=ZGradientBias})
   B=ScalarParameter() {defaultValue=35, parameterName=ZGradientDistance}
  )
  NamedRerouteDeclaration(None=ShaderStageSwitch(PixelShader=VertexInterpolator(VS=@n2).PS, VertexShader=@n2)) {name=GradientTopBottom, nodeColor=<rgba 0.6641 0 1 1>}
  PixelNormalWS()
  n3 = Fresnel(ExponentIn=ScalarParameter() {defaultValue=2.5663, parameterName=Fresnel})
  n4 = LinearInterpolate(
   A=VectorParameter() {parameterName=Color, defaultValue=<rgba 0.0633 0.1094 0 1>}.RGB
   B=VectorParameter() {parameterName=Color2, defaultValue=<rgba 0.1228 0.3906 0 1>}.RGB
   Alpha=Power(Base=NamedRerouteUsage(), Exp=ScalarParameter() {defaultValue=1.2755, parameterName=FalloffGradient})
  )
  n5 = TextureSample() {
   samplerType=SAMPLERTYPE_Masks
   texture=/BRCosmetics/Accessories/FORT_Backpacks/Backpack_M_MED_ShrimpStroll/Textures/T_ShrimpStroll_Leaves__FXMask
  }
  n6 = LinearInterpolate(B=ScalarParameter() {defaultValue=0.1572, parameterName=MInDarknessBody}, Alpha=@n5.R)
  n7 = VertexColor()
  n8 = LinearInterpolate(
   A=@n4
   B=Multiply(A=@n4, B=ScalarParameter() {defaultValue=0.44, parameterName=ColorNoise})
   Alpha=LinearInterpolate(A=Saturate(None=@n6), B=@n7.R, Alpha=@n7.R)
  )
  n9 = Multiply(A=@n8, B=ScalarParameter() {defaultValue=0.36, parameterName=FresnellBrightness})
  n10 = Multiply(A=Add(A=Multiply(A=@n3, B=@n9), B=@n8), B=@n5.G)
  Reroute(None=@n10)
  Reroute(None=LinearInterpolate(
    A=ScalarParameter() {defaultValue=1, parameterName=Roughness_low}
    B=ScalarParameter() {defaultValue=0.411, parameterName=Roughness}
    Alpha=@n5.G
   ))
  Reroute(None=ScalarParameter() {defaultValue=1, parameterName=SSSOpacity})
  n11 = Add(A=Distance(A=ObjectPositionWS(), B=CameraPositionWS()), B=ScalarParameter() {defaultValue=-100, parameterName=CameraDistanceFadeOffset})
  n12 = ScalarParameter() {defaultValue=500, parameterName=CameraDistanceFade}
  n13 = LinearInterpolate(
   A=ScalarParameter() {defaultValue=1.2594, parameterName=ThicknessNear}
   B=ScalarParameter() {parameterName=Thickness}
   Alpha=Saturate(None=Divide(A=@n11, B=@n12))
  )
  n14 = VertexNormalWS()
  n15 = Saturate(None=OneMinus(None=MaterialFunctionCall!(Thickness=@n13, NormalWS=@n14, PowerExponent=@n1) {materialFunction=/Game/Weapons/Prototype/ObjectMover/Materials/MF_NormalizedFresnel}.PowerInverted))
  n16 = MaterialFunctionCall(Transition Phase=@n15, Height Texture=@n5.G, Contrast=Constant()) {materialFunction=&HeightLerp}
  n17 = LinearInterpolate(A=@n5.B, Alpha=@n15) {constA=1}
  n18 = VertexColor()
  Reroute(None=LinearInterpolate(A=Multiply(A=@n16.Alpha, B=@n17), B=@n5.B, Alpha=@n18.G))
  n19 = TextureSample() {samplerType=SAMPLERTYPE_Normal, texture=/BRCosmetics/Accessories/FORT_Backpacks/Backpack_M_MED_ShrimpStroll/Textures/T_ShrimpStroll_Leaves_N}
  n20 = MaterialFunctionCall(Normal=@n19.RGB, Flatness=ScalarParameter() {defaultValue=0.416, parameterName=LeafNormal}) {materialFunction=&FlattenNormal}
  Reroute(None=Multiply(A=LinearInterpolate(A=@n19.RGB, B=@n20.Result, Alpha=@n18.G), B=TwoSidedSign()))
  n21 = VertexColor()
  n22 = VertexNormalWS()
  n23 = Add(
   A=Divide(A=LocalPosition() {localOrigin=InstancePreSkinning}.XYZ, B=ScalarParameter() {defaultValue=2, parameterName=SinewaveFrequency})
   B=Multiply(A=Time(), B=ScalarParameter() {defaultValue=0.3, parameterName=Speed})
  )
  n24 = Add(
   A=Divide(A=LocalPosition() {localOrigin=InstancePreSkinning}.XYZ, B=ScalarParameter() {defaultValue=4, parameterName=SinewaveFrequency2})
   B=Multiply(A=Time(), B=ScalarParameter() {defaultValue=0.712, parameterName=Speed2})
  )
  n25 = Multiply(A=Add(A=Sine(None=@n23)), B=Add(A=Sine(None=@n24)))
  n26 = Multiply(A=Multiply(A=Multiply(A=@n22, B=@n25), B=@n21.G), B=ScalarParameter() {defaultValue=0.1, parameterName=Amplitude})
  n27 = Multiply(A=Time(), B=ScalarParameter() {defaultValue=0.12, parameterName=SinewaveFrequency3})
  n28 = Sine(None=@n27)
  n29 = Multiply(
   A=MaterialFunctionCall(R=Cosine(None=@n27), G=@n28, B=Constant()) {materialFunction=&AppendMany}.RGB
   B=Multiply(A=ScalarParameter() {defaultValue=0.5, parameterName=MovementTopAmplitude}, B=NamedRerouteUsage())
  )
  Reroute(None=Add(A=Multiply(A=@n21.B, B=@n26), B=Multiply(A=@n29, B=@n28)))
  Reroute(None=Multiply(A=@n10, B=ScalarParameter() {parameterName=SSSColor}))
  ScalarParameter() {defaultValue=0.08, parameterName=Specular}
  TextureSample() {texture=/Game/Environments/Apollo/Foliage/Hedges/Textures/T_Apollo_Hedge_D}
  VectorParameter() {parameterName=FresnelColor, defaultValue=<rgba 0.1819 0.4479 0.0869 1>}
  VertexNormalWS()
 ]
}
