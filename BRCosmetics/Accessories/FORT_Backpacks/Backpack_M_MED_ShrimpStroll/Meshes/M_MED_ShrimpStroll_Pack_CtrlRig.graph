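# Control Rig for the ShrimpStroll backpack cosmetic, adding procedural secondary motion to its "hedge" bones (base, tail, arms, head).
# On execution, a ModifyTransforms unit applies a small fixed rotation to dyn_hedge_base, then feeds execution into a ChainHarmonics unit rooted at the pack bone.
# The harmonics combine pendulum lag with a sine-eased wave for sway, with reach disabled and debug drawing switched on.
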ControlRigBlueprint /BRCosmetics/Accessories/FORT_Backpacks/Backpack_M_MED_ShrimpStroll/Meshes/M_MED_ShrimpStroll_Pack_CtrlRig
graphs [
 RigVMFunctionLibrary {links [], nodes []}
 RigVMModel {nodes [
   RigUnit_BeginExecution = &RigUnit_BeginExecution()
   &RigUnit_ModifyTransforms(
    ItemToModify=((Item=(Type=Bone,Name="dyn_hedge_base"),Transform=(Rotation=(X=0,Y=-0.0436,Z=0,W=0.999),Translation=(X=0,Y=0,Z=0),Scale3D=(X=1,Y=1,Z=1))))
    Weight=1
    WeightMinimum=0
    ExecutePin=@RigUnit_BeginExecution.ExecutePin
   )
   > /Script/ControlRig.RigUnit_ChainHarmonicsPerItem(
    ChainRoot=(Type=Bone,Name="pack")
    DrawWorldOffset=(Rotation=(X=0,Y=0,Z=0,W=1),Translation=(X=0,Y=0,Z=0),Scale3D=(X=1,Y=1,Z=1))
    Pendulum=(bEnabled=True,PendulumStiffness=2,PendulumGravity=(X=0,Y=0,Z=0),PendulumBlend=0.75,PendulumDrag=0.98,PendulumMinimum=0,PendulumMaximum=0.1,PendulumEase=Linear,UnwindAxis=(X=0,Y=-1,Z=0),UnwindMinimum=0.2,UnwindMaximum=0.05)
    Reach=(bEnabled=False,ReachTarget=(X=0,Y=0,Z=0),ReachAxis=(X=1,Y=0,Z=0),ReachMinimum=0,ReachMaximum=0,ReachEase=Linear)
    Speed=(X=0,Y=0.8,Z=0.8)
    Wave=(bEnabled=True,WaveFrequency=(X=0.3,Y=0.3,Z=0.3),WaveAmplitude=(X=0,Y=-0.2,Z=0.2),WaveOffset=(X=0,Y=-2,Z=1),WaveNoise=(X=0,Y=0,Z=0),WaveMinimum=0,WaveMaximum=1,WaveEase=SineEaseInOut)
    WaveCurve=(EditorCurveData=(Keys=((),(Time=1,Value=1))))
    bDrawDebug=True
   ) {is=struct RigVMUnitNode}
  ]}
]
hierarchy [
 root {loc=<0 0 0>, rot=<0 0 0 1>}
 pack {parent=root, loc=<0 -0.7316 0>, rot=<0 0 0 1>}
 dyn_hedge_base {parent=pack, loc=<-0.2569 -6.3444 -17.9645>, rot=<-0.5 0.5 -0.5 -0.5>}
 dyn_hedge_tail {parent=dyn_hedge_base, loc=<9.3638 6.3918 0.009>, rot=<-0.601 -0.3726 -0.3726 -0.601>}
 dyn_hedge_arm_l {parent=dyn_hedge_base, loc=<20.5275 -4.1447 4.6581>, rot=<0.49 -0.5098 0.5697 -0.4188>}
 dyn_hedge_arm_r {parent=dyn_hedge_base, loc=<20.5308 -3.9971 -4.7356>, rot=<0.4662 -0.523 0.4855 -0.5229>}
 dyn_hedge_head {parent=dyn_hedge_base, loc=<24.9418 -4.7317 -0.0409>, rot=<0 0 0 1>}
]
